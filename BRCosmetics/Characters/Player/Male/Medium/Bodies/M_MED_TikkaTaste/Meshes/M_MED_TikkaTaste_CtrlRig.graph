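# Control Rig for the TikkaTaste medium male player body that drives corrective deform bones and accessory bones.
# FortDeformRig runs first, then weighted parent constraints place ring, strap, zipper and boot-strap master bones from body bones.
# Spherical pose readers on both upperarms, smoothed by AlphaInterp, weight mirrored armor bone offsets, followed by fixed transform tweaks including boot straps.
# An earlier four-constraint ring chain and a standalone armor ModifyTransforms are left disconnected.
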ControlRigBlueprint /BRCosmetics/Characters/Player/Male/Medium/Bodies/M_MED_TikkaTaste/Meshes/M_MED_TikkaTaste_CtrlRig
graphs [
 RigVMFunctionLibrary {links [], nodes []}
 RigVMFunction_Sequence_1_ContainedGraph {nodes [
   n1 = RigVMFunctionEntryNode()
   RigVMFunction_Sequence = &RigVMFunction_Sequence(ExecuteContext=@n1.ExecuteContext)
   n2 = &RigVMFunction_Sequence(ExecuteContext=@RigVMFunction_Sequence.B)
   RigVMFunctionReturnNode(A=@RigVMFunction_Sequence.A, B=@n2.A, C=@n2.B)
  ]}
 RigVMModel {nodes [
   &RigUnit_ParentConstraint(
    AdvancedSettings=(RotationOrderForFilter=XZY)
    Child=(Type=Bone,Name="dyn_ring_back_out_r_MSTR")
    ExecutePin=()
    Parents="((Item=(Type=Bone,Name=\"spine_02\"),Weight=0.0152),(Item=(Type=Bone,Name=\"spine_05\"),Weight=0.0163),(Item=(Type=Bone,Name=\"upperarm_twist_02_r\"),Weight=0.0254),(Item=(Type=Bone,Name=\"spine_03\"),Weight=0.0835),(Item=(Type=Bone,Name=\"upperarm_twist_01_r\"),Wei ...(truncated)"
    Weight=1
   )
   > &RigUnit_ParentConstraint(
    AdvancedSettings=(RotationOrderForFilter=XZY)
    Child=(Type=Bone,Name="dyn_ring_back_in_r_MSTR")
    Parents="((Item=(Type=Bone,Name=\"spine_01\"),Weight=0.001),(Item=(Type=Bone,Name=\"upperarm_twist_01_r\"),Weight=0.0188),(Item=(Type=Bone,Name=\"deform_upperarm_bk_r\"),Weight=0.037),(Item=(Type=Bone,Name=\"spine_02\"),Weight=0.0415),(Item=(Type=Bone,Name=\"spine_05\"),Weig ...(truncated)"
    Weight=1
   )
   > &RigUnit_ParentConstraint(
    AdvancedSettings=(RotationOrderForFilter=XZY)
    Child=(Type=Bone,Name="dyn_ring_back_in_l_MSTR")
    Parents="((Item=(Type=Bone,Name=\"spine_01\"),Weight=0.001),(Item=(Type=Bone,Name=\"upperarm_twist_01_l\"),Weight=0.0188),(Item=(Type=Bone,Name=\"deform_upperarm_bk_l\"),Weight=0.037),(Item=(Type=Bone,Name=\"spine_02\"),Weight=0.0415),(Item=(Type=Bone,Name=\"spine_05\"),Weig ...(truncated)"
    Weight=1
   )
   > &RigUnit_ParentConstraint(
    AdvancedSettings=(RotationOrderForFilter=XZY)
    Child=(Type=Bone,Name="dyn_ring_back_out_l_MSTR")
    Parents="((Item=(Type=Bone,Name=\"spine_02\"),Weight=0.0152),(Item=(Type=Bone,Name=\"spine_05\"),Weight=0.0163),(Item=(Type=Bone,Name=\"upperarm_twist_02_l\"),Weight=0.0254),(Item=(Type=Bone,Name=\"spine_03\"),Weight=0.0835),(Item=(Type=Bone,Name=\"upperarm_twist_01_l\"),Wei ...(truncated)"
    Weight=1
   )
   RigUnit_BeginExecution = &RigUnit_BeginExecution()
   RigVMFunction_Sequence = RigVMAggregateNode(ExecuteContext=@RigUnit_BeginExecution.ExecutePin)
   /Script/FortniteGame.RigUnit_FortDeformRig(
    Bool=(bUpperArm=true,bElbow=true,bWrist=true,bGroin=true,bGlute=true,bKnee=true,bPec=true,bBicep=False,bLat=False)
    Curve="(bLUpperArmUpAlpha=False,bRUpperArmUpAlpha=False,bLUpperArmFrAlpha=False,bRUpperArmFrAlpha=False,bLUpperArmBkAlpha=False,bRUpperArmBkAlpha=False,bLWristUpAlpha=False,bRWristUpAlpha=False,bLWristDnAlpha=False,bRWristDnAlpha=False,bLKneeAlpha=False,bRKneeAlp ...(truncated)"
    Debug=(bUpperArmDebug=False,bWristDebug=False,bLatDebug=False,bGroinGluteDebug=False)
    Value="(UpperArmUp=(Rotation=(X=0.0,Y=-0.0,Z=0.0,W=1.0),Translation=(X=0.0,Y=0.0,Z=8.0),Scale3D=(X=0.70,Y=1.0,Z=1.0)),UpperArmFwdFr=(Rotation=(X=0.0,Y=-0.0,Z=0.0,W=1.0),Translation=(X=0.0,Y=10.0,Z=0.0),Scale3D=(X=0.50,Y=1.0,Z=1.0)),UpperArmFwdBk=(Rotation=(X=0.0, ...(truncated)"
    ExecutePin=@RigVMFunction_Sequence.A
   ) {is=struct RigVMUnitNode}
   > &RigUnit_ParentConstraint(
    AdvancedSettings=(RotationOrderForFilter=XZY)
    Child=(Type=Bone,Name="dyn_ring_back_out_l_MSTR")
    Parents="((Item=(Type=Bone,Name=\"spine_02\"),Weight=0.0151),(Item=(Type=Bone,Name=\"spine_05\"),Weight=0.018),(Item=(Type=Bone,Name=\"upperarm_twist_02_l\"),Weight=0.0255),(Item=(Type=Bone,Name=\"spine_03\"),Weight=0.0833),(Item=(Type=Bone,Name=\"upperarm_twist_01_l\"),Weig ...(truncated)"
    Weight=1
   )
   > &RigUnit_ParentConstraint(
    AdvancedSettings=(RotationOrderForFilter=XZY)
    Child=(Type=Bone,Name="dyn_ring_back_in_l_MSTR")
    Parents="((Item=(Type=Bone,Name=\"upperarm_twist_01_l\"),Weight=0.023),(Item=(Type=Bone,Name=\"deform_upperarm_bk_l\"),Weight=0.0376),(Item=(Type=Bone,Name=\"spine_02\"),Weight=0.0382),(Item=(Type=Bone,Name=\"spine_05\"),Weight=0.1015),(Item=(Type=Bone,Name=\"spine_04\"),Wei ...(truncated)"
    Weight=1
   )
   > &RigUnit_ParentConstraint(
    AdvancedSettings=(RotationOrderForFilter=XZY)
    Child=(Type=Bone,Name="dyn_ring_back_in_r_MSTR")
    Parents="((Item=(Type=Bone,Name=\"upperarm_twist_01_r\"),Weight=0.023),(Item=(Type=Bone,Name=\"deform_upperarm_bk_r\"),Weight=0.0376),(Item=(Type=Bone,Name=\"spine_02\"),Weight=0.0382),(Item=(Type=Bone,Name=\"spine_05\"),Weight=0.1015),(Item=(Type=Bone,Name=\"spine_04\"),Wei ...(truncated)"
    Weight=1
   )
   > &RigUnit_ParentConstraint(
    AdvancedSettings=(RotationOrderForFilter=XZY)
    Child=(Type=Bone,Name="dyn_ring_back_out_r_MSTR")
    Parents="((Item=(Type=Bone,Name=\"spine_02\"),Weight=0.0151),(Item=(Type=Bone,Name=\"spine_05\"),Weight=0.018),(Item=(Type=Bone,Name=\"upperarm_twist_02_r\"),Weight=0.0255),(Item=(Type=Bone,Name=\"spine_03\"),Weight=0.0833),(Item=(Type=Bone,Name=\"upperarm_twist_01_r\"),Weig ...(truncated)"
    Weight=1
   )
   > &RigUnit_ParentConstraint(
    AdvancedSettings=(RotationOrderForFilter=XZY)
    Child=(Type=Bone,Name="dyn_chest_ring_l_MSTR")
    Parents="((Item=(Type=Bone,Name=\"spine_01\"),Weight=0.04),(Item=(Type=Bone,Name=\"deform_upperarm_fr_l\"),Weight=0.0446),(Item=(Type=Bone,Name=\"upperarm_twist_01_l\"),Weight=0.0817),(Item=(Type=Bone,Name=\"spine_02\"),Weight=0.1335),(Item=(Type=Bone,Name=\"clavicle_l\"),We ...(truncated)"
    Weight=1
   )
   > &RigUnit_ParentConstraint(
    AdvancedSettings=(RotationOrderForFilter=XZY)
    Child=(Type=Bone,Name="dyn_chest_ring_r_MSTR")
    Parents="((Item=(Type=Bone,Name=\"spine_01\"),Weight=0.04),(Item=(Type=Bone,Name=\"deform_upperarm_fr_r\"),Weight=0.0446),(Item=(Type=Bone,Name=\"upperarm_twist_01_r\"),Weight=0.0817),(Item=(Type=Bone,Name=\"spine_02\"),Weight=0.1335),(Item=(Type=Bone,Name=\"clavicle_r\"),We ...(truncated)"
    Weight=1
   )
   > &RigUnit_ParentConstraint(
    AdvancedSettings=(RotationOrderForFilter=XZY)
    Child=(Type=Bone,Name="dyn_chest_strap_top_l_MSTR")
    Parents="((Item=(Type=Bone,Name=\"deform_upperarm_bk_l\"),Weight=0.0289),(Item=(Type=Bone,Name=\"upperarm_twist_01_l\"),Weight=0.0382),(Item=(Type=Bone,Name=\"pelvis\"),Weight=0.0533),(Item=(Type=Bone,Name=\"clavicle_l\"),Weight=0.101),(Item=(Type=Bone,Name=\"spine_01\"),Wei ...(truncated)"
    Weight=1
   )
   > &RigUnit_ParentConstraint(
    AdvancedSettings=(RotationOrderForFilter=XZY)
    Child=(Type=Bone,Name="dyn_chest_strap_mid_l_MSTR")
    Parents="((Item=(Type=Bone,Name=\"deform_upperarm_bk_l\"),Weight=0.0035),(Item=(Type=Bone,Name=\"upperarm_twist_01_l\"),Weight=0.0049),(Item=(Type=Bone,Name=\"clavicle_l\"),Weight=0.0405),(Item=(Type=Bone,Name=\"spine_04\"),Weight=0.0681),(Item=(Type=Bone,Name=\"pelvis\"),We ...(truncated)"
    Weight=1
   )
   > &RigUnit_ParentConstraint(
    AdvancedSettings=(RotationOrderForFilter=XZY)
    Child=(Type=Bone,Name="dyn_chest_strap_btm_l_MSTR")
    Parents="((Item=(Type=Bone,Name=\"clavicle_l\"),Weight=0.0054),(Item=(Type=Bone,Name=\"spine_04\"),Weight=0.0213),(Item=(Type=Bone,Name=\"spine_03\"),Weight=0.2033),(Item=(Type=Bone,Name=\"spine_01\"),Weight=0.2089),(Item=(Type=Bone,Name=\"pelvis\"),Weight=0.2449),(Item=(Typ ...(truncated)"
    Weight=1
   )
   > &RigUnit_ParentConstraint(
    AdvancedSettings=(RotationOrderForFilter=XZY)
    Child=(Type=Bone,Name="dyn_chest_strap_top_r_MSTR")
    Parents="((Item=(Type=Bone,Name=\"deform_upperarm_bk_r\"),Weight=0.0289),(Item=(Type=Bone,Name=\"upperarm_twist_01_r\"),Weight=0.0382),(Item=(Type=Bone,Name=\"pelvis\"),Weight=0.0533),(Item=(Type=Bone,Name=\"clavicle_r\"),Weight=0.101),(Item=(Type=Bone,Name=\"spine_01\"),Wei ...(truncated)"
    Weight=1
   )
   > &RigUnit_ParentConstraint(
    AdvancedSettings=(RotationOrderForFilter=XZY)
    Child=(Type=Bone,Name="dyn_chest_strap_mid_r_MSTR")
    Parents="((Item=(Type=Bone,Name=\"deform_upperarm_bk_r\"),Weight=0.0035),(Item=(Type=Bone,Name=\"upperarm_twist_01_r\"),Weight=0.0049),(Item=(Type=Bone,Name=\"clavicle_r\"),Weight=0.0405),(Item=(Type=Bone,Name=\"spine_04\"),Weight=0.0681),(Item=(Type=Bone,Name=\"pelvis\"),We ...(truncated)"
    Weight=1
   )
   > &RigUnit_ParentConstraint(
    AdvancedSettings=(RotationOrderForFilter=XZY)
    Child=(Type=Bone,Name="dyn_chest_strap_btm_r_MSTR")
    Parents="((Item=(Type=Bone,Name=\"clavicle_r\"),Weight=0.0054),(Item=(Type=Bone,Name=\"spine_04\"),Weight=0.0213),(Item=(Type=Bone,Name=\"spine_03\"),Weight=0.2033),(Item=(Type=Bone,Name=\"spine_01\"),Weight=0.2089),(Item=(Type=Bone,Name=\"pelvis\"),Weight=0.2449),(Item=(Typ ...(truncated)"
    Weight=1
   )
   > &RigUnit_ParentConstraint(
    AdvancedSettings=(RotationOrderForFilter=XZY)
    Child=(Type=Bone,Name="dyn_zipper_l_MSTR")
    Parents=((Item=(Type=Bone,Name="thigh_twist_01_l"),Weight=0.624),(Item=(Type=Bone,Name="pelvis"),Weight=0.113),(Item=(Type=Bone,Name="deform_groin_l"),Weight=0.221),(Item=(Type=Bone,Name="thigh_l"),Weight=0.04))
    Weight=1
   )
   > &RigUnit_ParentConstraint(
    AdvancedSettings=(RotationOrderForFilter=XZY)
    Child=(Type=Bone,Name="dyn_boot_strap_l_MSTR")
    Parents=((Item=(Type=Bone,Name="calf_l"),Weight=0.0049),(Item=(Type=Bone,Name="calf_twist_02_l"),Weight=0.4077),(Item=(Type=Bone,Name="calf_twist_01_l"),Weight=0.5874))
    Weight=1
   )
   > &RigUnit_ParentConstraint(
    AdvancedSettings=(RotationOrderForFilter=XZY)
    Child=(Type=Bone,Name="dyn_boot_strap_r_MSTR")
    Parents=((Item=(Type=Bone,Name="calf_r"),Weight=0.0049),(Item=(Type=Bone,Name="calf_twist_02_r"),Weight=0.4077),(Item=(Type=Bone,Name="calf_twist_01_r"),Weight=0.5874))
    Weight=1
   )
   > &RigUnit_ParentConstraint(
    AdvancedSettings=(RotationOrderForFilter=XZY)
    Child=(Type=Bone,Name="dyn_zipper_r_MSTR")
    Parents=((Item=(Type=Bone,Name="thigh_twist_01_r"),Weight=0.624),(Item=(Type=Bone,Name="pelvis"),Weight=0.113),(Item=(Type=Bone,Name="deform_groin_r"),Weight=0.221),(Item=(Type=Bone,Name="thigh_r"),Weight=0.04))
    Weight=1
   )
   n1 = &RigUnit_SphericalPoseReader(
    ActiveRegionScaleFactors=(PositiveWidth=1,NegativeWidth=1,PositiveHeight=1,NegativeHeight=1)
    ActiveRegionSize=0.125
    Debug=(bDrawDebug=True,bDraw2D=False,bDrawLocalAxes=False,DebugScale=25,DebugSegments=20,DebugThickness=0.25)
    DriverAxis=(X=1,Y=0,Z=0)
    DriverItem=(Type=Bone,Name="upperarm_l")
    FalloffRegionScaleFactors=(PositiveWidth=1,NegativeWidth=1,PositiveHeight=1,NegativeHeight=1)
    FalloffSize=0.4
    FlipHeightScaling=False
    FlipWidthScaling=False
    OptionalParentItem=(Type=Bone,Name="spine_05")
    RotationOffset=(X=52.479,Y=-24.8334,Z=29.047)
    ExecutePin=@RigVMFunction_Sequence.C
   )
   > n2 = &RigUnit_SphericalPoseReader(
    ActiveRegionScaleFactors=(PositiveWidth=1,NegativeWidth=1,PositiveHeight=1,NegativeHeight=1)
    ActiveRegionSize=0.125
    Debug=(bDrawDebug=True,bDraw2D=False,bDrawLocalAxes=False,DebugScale=25,DebugSegments=20,DebugThickness=0.25)
    DriverAxis=(X=-1,Y=0,Z=0)
    DriverItem=(Type=Bone,Name="upperarm_r")
    FalloffRegionScaleFactors=(PositiveWidth=1,NegativeWidth=1,PositiveHeight=1,NegativeHeight=1)
    FalloffSize=0.4
    FlipHeightScaling=False
    FlipWidthScaling=False
    OptionalParentItem=(Type=Bone,Name="spine_05")
    RotationOffset=(X=-100,Y=25,Z=30)
   )
   RigVMFunction_AlphaInterp = &RigVMFunction_AlphaInterp(InterpSpeedDecreasing=20, InterpSpeedIncreasing=20, Scale=1, bClampResult=False, bInterpResult=True, Value=@n1.OutputParam)
   n3 = &RigVMFunction_AlphaInterp(InterpSpeedDecreasing=20, InterpSpeedIncreasing=20, Scale=1, bClampResult=False, bInterpResult=True, Value=@n2.OutputParam)
   > Modify Transforms with Mirror = RigVMFunctionReferenceNode(
    Base Modify Data="((Item=(Type=Bone,Name=\"deform_armor_frnt_l\"),Transform=(Rotation=(X=0,Y=0,Z=0.0724,W=0.9974),Translation=(X=3.9,Y=-3.4,Z=6.2),Scale3D=(X=1,Y=1,Z=1))),(Item=(Type=Bone,Name=\"deform_armor_bk_l\"),Transform=(Rotation=(X=0.0125,Y=-0.1238,Z=-0.0989,W=0.9873),Tr ...(truncated)"
    Mirror Modify Data="((Item=(Type=Bone,Name=\"deform_armor_frnt_r\"),Transform=(Rotation=(X=0,Y=0,Z=0,W=1),Translation=(X=0,Y=0,Z=0),Scale3D=(X=1,Y=1,Z=1))),(Item=(Type=Bone,Name=\"deform_armor_bk_r\"),Transform=(Rotation=(X=0,Y=0,Z=0,W=1),Translation=(X=2,Y=0,Z=3),Scale3D=(X=1,Y= ...(truncated)"
    Weight=@RigVMFunction_AlphaInterp.Result
    "Mirror Weight "=@n3.Result
   )
   > &RigUnit_ModifyTransforms(
    ItemToModify="((Item=(Type=Bone,Name=\"dyn_boot_strap_l_MSTR\"),Transform=(Rotation=(X=0,Y=0,Z=-0.1736,W=0.9848),Translation=(X=0,Y=0,Z=0),Scale3D=(X=1,Y=1,Z=1))),(Item=(Type=Bone,Name=\"dyn_boot_strap_r_MSTR\"),Transform=(Rotation=(X=0,Y=0,Z=-0.1736,W=0.9848),Translation=( ...(truncated)"
    Weight=1
    WeightMinimum=0
   )
   &RigUnit_ModifyTransforms(
    ExecutePin=()
    ItemToModify="((Item=(Type=Bone,Name=\"deform_armor_frnt_l\"),Transform=(Rotation=(X=0,Y=0,Z=-0.0723,W=-0.9974),Translation=(X=3.8645,Y=-3.3512,Z=6.2109),Scale3D=(X=1,Y=1,Z=1))),(Item=(Type=Bone,Name=\"deform_armor_bk_l\"),Transform=(Rotation=(X=-0.0124,Y=0.1244,Z=0.0987,W= ...(truncated)"
    Weight=1
    WeightMinimum=0
   )
  ]}
]
hierarchy [
 root {loc=<0 0 0>, rot=<0 0 0 1>}
 pelvis {parent=root, loc=<0 -0.7071 93.1191>, rot=<0.0127 0.707 -0.0127 -0.707>}
 spine_01 {parent=pelvis, loc=<6.7494 -0.6388 0>, rot=<0 0 0.034 -0.9994>}
 spine_02 {parent=spine_01, loc=<7.2518 -0.035 0>, rot=<0 0 0.0238 -0.9997>}
 spine_03 {parent=spine_02, loc=<6.0104 0 0>, rot=<0 0 0.0149 -0.9999>}
 spine_04 {parent=spine_03, loc=<8.1024 0 0>, rot=<0 0 -0.0132 -0.9999>}
 spine_05 {parent=spine_04, loc=<10.0146 0 0>, rot=<0 0 -0.0015 -1>}
 neck_01 {parent=spine_05, loc=<11.3138 0.1529 0>, rot=<0 0 -0.2003 -0.9797>}
 neck_02 {parent=neck_01, loc=<4.5627 0 0>, rot=<0 0 0.0504 -0.9987>}
 head {parent=neck_02, loc=<5.3914 0 0>, rot=<0 0 0.115 -0.9934>}
 attach_cape {parent=spine_05, loc=<9.0169 -11.7132 -0.0003>, rot=<0.0616 -0.7044 -0.0616 -0.7044>}
 cape {parent=attach_cape, type=Null, loc=<0 0 0>, rot=<0 0 0 1>}
 attach_backpack {parent=spine_05, loc=<-0.1222 -13.472 -0.0002>, rot=<-0.0093 -0.707 0.0093 -0.707>}
 Backpack_BR {parent=attach_backpack, type=Null, loc=<0 0 0>, rot=<0 0 0 1>}
 clavicle_l {parent=spine_05, loc=<4.3575 -0.1466 -2.3223>, rot=<0.0465 -0.7289 -0.009 -0.6829>}
 upperarm_l {parent=clavicle_l, loc=<15.5111 0 0>, rot=<0.0226 -0.3234 -0.0088 -0.9459>}
 lowerarm_l {parent=upperarm_l, loc=<26.6514 0 0>, rot=<0.0391 -0.0099 -0.2453 -0.9686>}
 hand_l {parent=lowerarm_l, loc=<22.3491 0 0>, rot=<-0.7081 -0.0507 0.0975 -0.6975>}
 thumb_01_l {parent=hand_l, loc=<1.3118 -0.3191 -2.459>, rot=<0.6546 -0.4042 0.0664 -0.6354>}
 thumb_02_l {parent=thumb_01_l, loc=<5.9336 -0.0448 0.0401>, rot=<-0.0037 0.0213 0.298 -0.9543>}
 thumb_03_l {parent=thumb_02_l, loc=<2.2223 0.3682 -0.0907>, rot=<0.0118 -0.013 0.0066 -0.9998>}
 index_metacarpal_l {parent=hand_l, loc=<1.6535 0.0183 -2.96>, rot=<0.0158 -0.0124 -0.0173 -0.9996>}
 index_01_l {parent=index_metacarpal_l, loc=<8.3008 0.0357 -0.0524>, rot=<-0.0044 0.0485 0.0924 -0.9945>}
 index_02_l {parent=index_01_l, loc=<3.3587 -0.3004 0.0452>, rot=<-0.002 0.0027 0.3225 -0.9466>}
 index_03_l {parent=index_02_l, loc=<2.5578 -0.123 -0.0252>, rot=<-0.002 0.0067 0.2558 -0.9667>}
 middle_metacarpal_l {parent=hand_l, loc=<2.1856 -0.1725 -0.5984>, rot=<0.006 0.0168 -0.0329 -0.9993>}
 middle_01_l {parent=middle_metacarpal_l, loc=<7.3275 -0.2447 0.0129>, rot=<-0.0208 0.0042 0.0858 -0.9961>}
 middle_02_l {parent=middle_01_l, loc=<4.0123 0.1401 0.1175>, rot=<-0.0218 0.0258 0.3512 -0.9357>}
 middle_03_l {parent=middle_02_l, loc=<2.6077 0.0534 0.1373>, rot=<0.0082 0.0188 0.1708 -0.9851>}
 ring_metacarpal_l {parent=hand_l, loc=<2.2074 -0.3625 1.3638>, rot=<-0.0284 0.0722 -0.0458 -0.9959>}
 ring_01_l {parent=ring_metacarpal_l, loc=<7.0295 -0.6197 -0.1446>, rot=<-0.0489 -0.0252 0.1892 -0.9804>}
 ring_02_l {parent=ring_01_l, loc=<3.6174 0.4647 -0.1027>, rot=<0 0 0.2806 -0.9598>}
 ring_03_l {parent=ring_02_l, loc=<2.4152 0.114 0.0199>, rot=<0 0 0.1189 -0.9929>}
 pinky_metacarpal_l {parent=hand_l, loc=<1.7616 -0.9753 3.3814>, rot=<-0.0397 0.0814 -0.069 -0.9935>}
 pinky_01_l {parent=pinky_metacarpal_l, loc=<7.1709 -0.4831 -0.1094>, rot=<-0.0654 -0.0035 0.2242 -0.9723>}
 pinky_02_l {parent=pinky_01_l, loc=<2.8274 0.0893 -0.1043>, rot=<0.0037 0.0011 0.2854 -0.9584>}
 pinky_03_l {parent=pinky_02_l, loc=<2.2046 0.0307 -0.0078>, rot=<-0.0002 0.0098 0.1445 -0.9895>}
 deform_wrist_dn_base_l {parent=hand_l, loc=<0 -2.2342 0>, rot=<0 0 0 -1>}
 deform_wrist_dn_l {parent=deform_wrist_dn_base_l, loc=<0 0 0>, rot=<0 0 0 1>}
 deform_wrist_up_base_l {parent=hand_l, loc=<0 2.2682 0>, rot=<0 0 0 -1>}
 deform_wrist_up_l {parent=deform_wrist_up_base_l, loc=<0 0 0>, rot=<0 0 0 1>}
 lowerarm_twist_02_l {parent=lowerarm_l, loc=<9.6199 0 0>, rot=<0 0 0 -1>}
 lowerarm_twist_01_l {parent=lowerarm_l, loc=<17.8549 0 0>, rot=<0 0 0 -1>}
 deform_elbow_in_base_l {parent=lowerarm_l, loc=<0.2112 5.259 0.8274>, rot=<0 0 0 -1>}
 deform_elbow_in_l {parent=deform_elbow_in_base_l, loc=<0 0 0>, rot=<0 0 0 1>}
 deform_elbow_out_base_l {parent=lowerarm_l, loc=<0 -4 0>, rot=<0 0 0 -1>}
 deform_elbow_out_l {parent=deform_elbow_out_base_l, loc=<0 0 0>, rot=<0 0 0 1>}
 upperarm_twist_01_l {parent=upperarm_l, loc=<0.0041 0 0>, rot=<0 0 0 -1>}
 deform_upperarm_bk_base_l {parent=upperarm_twist_01_l, loc=<1.0813 -7.2193 -1.9453>, rot=<0 0 0 1>}
 deform_upperarm_bk_l {parent=deform_upperarm_bk_base_l, loc=<0 0 0>, rot=<0 0 0 1>}
 deform_upperarm_fr_base_l {parent=upperarm_twist_01_l, loc=<-0.3691 7.0506 -2.765>, rot=<0 0 0 1>}
 deform_upperarm_fr_l {parent=deform_upperarm_fr_base_l, loc=<0 0 0>, rot=<0 0 0 1>}
 deform_upperarm_up_base_l {parent=upperarm_twist_01_l, loc=<-3.2216 -1.1391 3.3311>, rot=<0 0 0 1>}
 deform_upperarm_up_l {parent=deform_upperarm_up_base_l, loc=<0 0 0>, rot=<0 0 0 1>}
 upperarm_twist_02_l {parent=upperarm_l, loc=<13.7941 0 0>, rot=<0 0 0 -1>}
 clavicle_r {parent=spine_05, loc=<4.3575 -0.1466 2.3223>, rot=<0.7289 0.0465 -0.6829 0.009>}
 upperarm_r {parent=clavicle_r, loc=<-15.5111 0 0>, rot=<0.0226 -0.3234 -0.0088 -0.9459>}
 lowerarm_r {parent=upperarm_r, loc=<-26.6514 0 0>, rot=<0.0391 -0.0099 -0.2453 -0.9686>}
 hand_r {parent=lowerarm_r, loc=<-22.3491 0 0>, rot=<-0.7081 -0.0507 0.0975 -0.6975>}
 thumb_01_r {parent=hand_r, loc=<-1.3118 0.3191 2.459>, rot=<0.6546 -0.4042 0.0664 -0.6354>}
 thumb_02_r {parent=thumb_01_r, loc=<-5.9336 0.0448 -0.0401>, rot=<-0.0037 0.0213 0.298 -0.9543>}
 thumb_03_r {parent=thumb_02_r, loc=<-2.2223 -0.3682 0.0907>, rot=<0.0118 -0.013 0.0066 -0.9998>}
 index_metacarpal_r {parent=hand_r, loc=<-1.6535 -0.0183 2.96>, rot=<0.0158 -0.0124 -0.0173 -0.9996>}
 index_01_r {parent=index_metacarpal_r, loc=<-8.3008 -0.0357 0.0524>, rot=<-0.0044 0.0485 0.0924 -0.9945>}
 index_02_r {parent=index_01_r, loc=<-3.3587 0.3004 -0.0452>, rot=<-0.002 0.0027 0.3225 -0.9466>}
 index_03_r {parent=index_02_r, loc=<-2.5578 0.123 0.0252>, rot=<-0.002 0.0067 0.2558 -0.9667>}
 middle_metacarpal_r {parent=hand_r, loc=<-2.1856 0.1725 0.5984>, rot=<0.006 0.0168 -0.0329 -0.9993>}
 middle_01_r {parent=middle_metacarpal_r, loc=<-7.3275 0.2447 -0.0129>, rot=<-0.0208 0.0042 0.0858 -0.9961>}
 middle_02_r {parent=middle_01_r, loc=<-4.0123 -0.1401 -0.1175>, rot=<-0.0218 0.0258 0.3512 -0.9357>}
 middle_03_r {parent=middle_02_r, loc=<-2.6077 -0.0534 -0.1373>, rot=<0.0082 0.0188 0.1708 -0.9851>}
 ring_metacarpal_r {parent=hand_r, loc=<-2.2074 0.3625 -1.3638>, rot=<-0.0284 0.0722 -0.0458 -0.9959>}
 ring_01_r {parent=ring_metacarpal_r, loc=<-7.0295 0.6197 0.1446>, rot=<-0.0489 -0.0252 0.1892 -0.9804>}
 ring_02_r {parent=ring_01_r, loc=<-3.6174 -0.4647 0.1027>, rot=<0 0 0.2806 -0.9598>}
 ring_03_r {parent=ring_02_r, loc=<-2.4152 -0.114 -0.0199>, rot=<0 0 0.1189 -0.9929>}
 pinky_metacarpal_r {parent=hand_r, loc=<-1.7616 0.9753 -3.3814>, rot=<-0.0397 0.0814 -0.069 -0.9935>}
 pinky_01_r {parent=pinky_metacarpal_r, loc=<-7.1709 0.4831 0.1094>, rot=<-0.0654 -0.0035 0.2242 -0.9723>}
 pinky_02_r {parent=pinky_01_r, loc=<-2.8274 -0.0893 0.1043>, rot=<0.0037 0.0011 0.2854 -0.9584>}
 pinky_03_r {parent=pinky_02_r, loc=<-2.2046 -0.0307 0.0078>, rot=<-0.0002 0.0098 0.1445 -0.9895>}
 deform_wrist_dn_base_r {parent=hand_r, loc=<-0.0004 2.2345 -0.0001>, rot=<0 0 0 -1>}
 deform_wrist_dn_r {parent=deform_wrist_dn_base_r, loc=<0 0 0>, rot=<0 0 0 1>}
 deform_wrist_up_base_r {parent=hand_r, loc=<0.0003 -2.2684 0.0001>, rot=<0 0 0 -1>}
 deform_wrist_up_r {parent=deform_wrist_up_base_r, loc=<0 0 0>, rot=<0 0 0 1>}
 lowerarm_twist_02_r {parent=lowerarm_r, loc=<-9.6199 0 0>, rot=<0 0 0 -1>}
 lowerarm_twist_01_r {parent=lowerarm_r, loc=<-17.8549 0 0>, rot=<0 0 0 -1>}
 deform_elbow_in_base_r {parent=lowerarm_r, loc=<-0.2111 -5.2591 -0.8275>, rot=<0 0 0 -1>}
 deform_elbow_in_r {parent=deform_elbow_in_base_r, loc=<-0.0001 0 0.0002>, rot=<0 0 0 1>}
 deform_elbow_out_base_r {parent=lowerarm_r, loc=<0 4 -0.0001>, rot=<0 0 0 -1>}
 deform_elbow_out_r {parent=deform_elbow_out_base_r, loc=<0 0 0>, rot=<0 0 0 1>}
 upperarm_twist_01_r {parent=upperarm_r, loc=<-0.0041 0 0>, rot=<0 0 0 -1>}
 deform_upperarm_bk_base_r {parent=upperarm_twist_01_r, loc=<-1.0814 7.2192 1.9455>, rot=<0 0 0 1>}
 deform_upperarm_bk_r {parent=deform_upperarm_bk_base_r, loc=<0 0 0>, rot=<0 0 0 1>}
 deform_upperarm_fr_base_r {parent=upperarm_twist_01_r, loc=<0.3689 -7.0506 2.7652>, rot=<0 0 0 1>}
 deform_upperarm_fr_r {parent=deform_upperarm_fr_base_r, loc=<0 0 0>, rot=<0 0 0 1>}
 deform_upperarm_up_base_r {parent=upperarm_twist_01_r, loc=<3.2214 1.1391 -3.3308>, rot=<0 0 0 1>}
 deform_upperarm_up_r {parent=deform_upperarm_up_base_r, loc=<0 0 0>, rot=<0 0 0 -1>}
 upperarm_twist_02_r {parent=upperarm_r, loc=<-13.7941 0 0>, rot=<0 0 0 -1>}
 deform_armor_bk_l {parent=spine_05, loc=<14.4496 -1.8086 -12.9099>, rot=<0.0839 -0.2128 -0.1727 -0.958>}
 deform_armor_frnt_l {parent=spine_05, loc=<5.6208 9.5619 -11.3561>, rot=<-0.0174 -0.1624 -0.4402 -0.8829>}
 deform_armor_bk_r {parent=spine_05, loc=<14.4494 -1.8086 12.9099>, rot=<-0.2128 -0.0839 0.958 -0.1727>}
 deform_armor_frnt_r {parent=spine_05, loc=<5.6208 9.5619 11.3561>, rot=<-0.1624 0.0174 0.8829 -0.4402>}
 deform_armor_back {parent=spine_05, loc=<17.4723 -6.51 0>, rot=<0.9774 -0.2114 0 0>}
 dyn_ring_back_out_l_MSTR {parent=spine_05, loc=<-1.3302 -11.3975 -15.6547>, rot=<0.5077 -0.5792 -0.4764 -0.4241>}
 dyn_ring_back_out_l {parent=dyn_ring_back_out_l_MSTR, loc=<0 0 0>, rot=<0 0 0 1>}
 dyn_ring_back_in_l_MSTR {parent=spine_05, loc=<-3.7708 -13.1422 -9.4777>, rot=<0.0809 -0.9931 -0.0849 0.0013>}
 dyn_ring_back_in_l {parent=dyn_ring_back_in_l_MSTR, loc=<0 0 0>, rot=<0 0 0 1>}
 dyn_ring_back_in_r_MSTR {parent=spine_05, loc=<-3.771 -13.1422 9.4777>, rot=<0.9931 0.0809 0.0013 0.0849>}
 dyn_ring_back_in_r {parent=dyn_ring_back_in_r_MSTR, loc=<0 0 0>, rot=<0 0 0 1>}
 dyn_ring_back_out_r_MSTR {parent=spine_05, loc=<-1.3299 -11.3975 15.6547>, rot=<0.5792 0.5077 -0.4241 0.4764>}
 dyn_ring_back_out_r {parent=dyn_ring_back_out_r_MSTR, loc=<0 0 0>, rot=<0 0 0 1>}
 dyn_chest_ring_l_MSTR {parent=spine_05, loc=<-9.1188 7.2513 -14.854>, rot=<-0.5667 0.5088 0.5731 0.3026>}
 dyn_chest_ring_l {parent=dyn_chest_ring_l_MSTR, loc=<0 0 0>, rot=<0 0 0 1>}
 dyn_chest_ring_r_MSTR {parent=spine_05, loc=<-9.1184 7.2513 14.854>, rot=<-0.5088 -0.5667 0.3026 -0.5731>}
 dyn_chest_ring_r {parent=dyn_chest_ring_r_MSTR, loc=<0 0 0>, rot=<0 0 0 1>}
 deform_chest_acc_MSTR {parent=spine_05, loc=<-0.5789 14.5888 -0.0109>, rot=<0.0284 -0.7065 -0.0284 -0.7065>}
 deform_chest_acc {parent=deform_chest_acc_MSTR, loc=<0 0 0>, rot=<0 0 0 1>}
 deform_pec_aim_base_l {parent=spine_04, loc=<7.8732 12.6388 -2.4123>, rot=<0.0102 -0.4602 0.3979 -0.7936>}
 deform_pec_aim_l {parent=deform_pec_aim_base_l, loc=<0 0 0>, rot=<0 0 0 1>}
 deform_pec_base_l {parent=deform_pec_aim_l, loc=<10.766 0 0>, rot=<0 0 0 1>}
 deform_pec_l {parent=deform_pec_base_l, loc=<0.055 0 0>, rot=<0 0 0 1>}
 deform_pec_aim_base_r {parent=spine_04, loc=<7.8727 12.6388 2.4123>, rot=<-0.4602 -0.0102 0.7936 0.3979>}
 deform_pec_aim_r {parent=deform_pec_aim_base_r, loc=<0 0 0>, rot=<0 0 0 1>}
 deform_pec_base_r {parent=deform_pec_aim_r, loc=<-10.7662 -0.0006 -0.0007>, rot=<0 0 0 1>}
 deform_pec_r {parent=deform_pec_base_r, loc=<0 0 0>, rot=<0 0 0 1>}
 deform_mag_top_l {parent=spine_03, loc=<6.9971 10.187 -13.898>, rot=<0.0387 -0.706 -0.0387 -0.706>}
 deform_mag_top_r {parent=spine_03, loc=<6.9971 10.187 13.898>, rot=<0.0387 -0.706 -0.0387 -0.706>}
 dyn_chest_strap_top_l_MSTR {parent=spine_02, loc=<9.3169 -0.3504 -16.0472>, rot=<0.6973 -0.6707 -0.2474 -0.0529>}
 dyn_chest_strap_top_l_01 {parent=dyn_chest_strap_top_l_MSTR, loc=<0 0 0>, rot=<0 0 0 1>}
 dyn_chest_strap_top_l_02 {parent=dyn_chest_strap_top_l_01, loc=<1.5981 0 0>, rot=<0 0 0 1>}
 dyn_chest_strap_mid_l_MSTR {parent=spine_02, loc=<5.6454 -0.6045 -15.2136>, rot=<0.697 -0.6615 -0.2616 -0.09>}
 dyn_chest_strap_mid_l_01 {parent=dyn_chest_strap_mid_l_MSTR, loc=<0 0 0>, rot=<0 0 0 1>}
 dyn_chest_strap_mid_l_02 {parent=dyn_chest_strap_mid_l_01, loc=<1.4614 0 0>, rot=<0 0 0 1>}
 dyn_chest_strap_btm_l_MSTR {parent=spine_02, loc=<1.9403 -0.3529 -14.6837>, rot=<0.7042 -0.6593 -0.2317 -0.1256>}
 dyn_chest_strap_btm_l_01 {parent=dyn_chest_strap_btm_l_MSTR, loc=<0 0 0>, rot=<0 0 0 1>}
 dyn_chest_strap_btm_l_02 {parent=dyn_chest_strap_btm_l_01, loc=<1.327 0 0>, rot=<0 0 0 1>}
 dyn_chest_strap_top_r_MSTR {parent=spine_02, loc=<9.3166 -0.3504 16.0472>, rot=<-0.6707 -0.6973 0.0529 -0.2474>}
 dyn_chest_strap_top_r_01 {parent=dyn_chest_strap_top_r_MSTR, loc=<0 0 0>, rot=<0 0 0 1>}
 dyn_chest_strap_top_r_02 {parent=dyn_chest_strap_top_r_01, loc=<-1.5982 -0.0002 -0.0001>, rot=<0 0 0 1>}
 dyn_chest_strap_mid_r_MSTR {parent=spine_02, loc=<5.6452 -0.6045 15.2136>, rot=<-0.6615 -0.697 0.09 -0.2616>}
 dyn_chest_strap_mid_r_01 {parent=dyn_chest_strap_mid_r_MSTR, loc=<0 0 0>, rot=<0 0 0 1>}
 dyn_chest_strap_mid_r_02 {parent=dyn_chest_strap_mid_r_01, loc=<-1.4614 0.0005 0.0001>, rot=<0 0 0 1>}
 dyn_chest_strap_btm_r_MSTR {parent=spine_02, loc=<1.9404 -0.3529 14.6837>, rot=<-0.6593 -0.7042 0.1256 -0.2317>}
 dyn_chest_strap_btm_r_01 {parent=dyn_chest_strap_btm_r_MSTR, loc=<0 0 0>, rot=<0 0 0 1>}
 dyn_chest_strap_btm_r_02 {parent=dyn_chest_strap_btm_r_01, loc=<-1.327 0.0003 0>, rot=<0 0 0 -1>}
 thigh_l {parent=pelvis, loc=<-2.5693 -0.0109 -7.7501>, rot=<0.0012 0.0572 0.0155 -0.9982>}
 calf_l {parent=thigh_l, loc=<-39.0484 0 0>, rot=<0.0006 -0.0161 -0.0391 -0.9991>}
 foot_l {parent=calf_l, loc=<-42.561 0.1077 0.4292>, rot=<0.0418 -0.0455 0.0343 -0.9975>}
 ball_l {parent=foot_l, loc=<-6.1599 13.8264 0.6806>, rot=<0.0001 -0.0001 -0.7063 -0.7079>}
 calf_twist_02_l {parent=calf_l, loc=<-18.8064 -0.0967 0.0397>, rot=<0.0237 0.0095 -0.0035 -0.9997>}
 calf_twist_01_l {parent=calf_l, loc=<-35.813 -0.0967 0.0397>, rot=<0.0237 0.0095 -0.0035 -0.9997>}
 dyn_boot_strap_l_MSTR {parent=calf_twist_01_l, loc=<5.096 -3.3565 5.0142>, rot=<0.5766 -0.811 0.0304 0.0941>}
 dyn_boot_strap_l_01 {parent=dyn_boot_strap_l_MSTR, loc=<0 0 0>, rot=<0 0 0 1>}
 dyn_boot_strap_l_02 {parent=dyn_boot_strap_l_01, loc=<1.4008 0 0>, rot=<0.0164 0.0535 -0.0386 -0.9977>}
 dyn_boot_strap_l_03 {parent=dyn_boot_strap_l_02, loc=<1.4164 0 0>, rot=<0 0.1045 0 -0.9945>}
 deform_knee_in_base_l {parent=calf_l, loc=<1.298 -5.2957 -0.0496>, rot=<0 0 0 -1>}
 deform_knee_in_l {parent=deform_knee_in_base_l, loc=<0 0 0>, rot=<0 0 0 1>}
 deform_knee_out_base_l {parent=calf_l, loc=<0 4 0>, rot=<0 0 0 -1>}
 deform_knee_out_l {parent=deform_knee_out_base_l, loc=<0 0 0>, rot=<0 0 0 1>}
 thigh_twist_01_l {parent=thigh_l, loc=<-21.6303 0 0>, rot=<0.0013 0 0 -1>}
 dyn_zipper_l_MSTR {parent=thigh_twist_01_l, loc=<11.0348 11.6856 -3.2519>, rot=<0.3507 -0.5305 0.6512 0.4141>}
 dyn_zipper_l {parent=dyn_zipper_l_MSTR, loc=<0 0 0>, rot=<0 0 0 1>}
 deform_groin_base_l {parent=thigh_l, loc=<-2.1908 8.1142 -1.7442>, rot=<0 0 0 -1>}
 deform_groin_l {parent=deform_groin_base_l, loc=<0 0 0>, rot=<0 0 0 1>}
 thigh_r {parent=pelvis, loc=<-2.5693 -0.0109 7.7501>, rot=<0.0572 -0.0012 0.9982 0.0155>}
 calf_r {parent=thigh_r, loc=<39.0484 0 0>, rot=<0.0006 -0.0161 -0.0391 -0.9991>}
 foot_r {parent=calf_r, loc=<42.561 -0.1077 -0.4292>, rot=<0.0418 -0.0455 0.0343 -0.9975>}
 ball_r {parent=foot_r, loc=<6.1599 -13.8264 -0.6806>, rot=<0.0001 -0.0001 -0.7063 -0.7079>}
 calf_twist_02_r {parent=calf_r, loc=<18.8064 0.0967 -0.0397>, rot=<0.0237 0.0095 -0.0035 -0.9997>}
 calf_twist_01_r {parent=calf_r, loc=<35.813 0.0967 -0.0397>, rot=<0.0237 0.0095 -0.0035 -0.9997>}
 dyn_boot_strap_r_MSTR {parent=calf_twist_01_r, loc=<-5.096 3.3565 -5.0142>, rot=<0.5766 -0.811 0.0304 0.0941>}
 dyn_boot_strap_r_01 {parent=dyn_boot_strap_r_MSTR, loc=<0 0 0>, rot=<0 0 0 1>}
 dyn_boot_strap_r__02 {parent=dyn_boot_strap_r_01, loc=<-1.4008 0 0>, rot=<0.0164 0.0535 -0.0386 -0.9977>}
 dyn_boot_strap_r__03 {parent=dyn_boot_strap_r__02, loc=<-1.4164 0 0>, rot=<0 0.1045 0 -0.9945>}
 deform_knee_in_base_r {parent=calf_r, loc=<-1.298 5.2957 0.0496>, rot=<0 0 0 -1>}
 deform_knee_in_r {parent=deform_knee_in_base_r, loc=<0 0 0>, rot=<0 0 0 1>}
 deform_knee_out_base_r {parent=calf_r, loc=<0 -4 0>, rot=<0 0 0 -1>}
 deform_knee_out_r {parent=deform_knee_out_base_r, loc=<0 0 0>, rot=<0 0 0 1>}
 thigh_twist_01_r {parent=thigh_r, loc=<21.6303 0 0>, rot=<0.0013 0 0 -1>}
 dyn_zipper_r_MSTR {parent=thigh_twist_01_r, loc=<-11.0347 -11.6857 3.2519>, rot=<0.3507 -0.5305 0.6512 0.4141>}
 dyn_zipper_r {parent=dyn_zipper_r_MSTR, loc=<0 0 0>, rot=<0 0 0 1>}
 deform_groin_base_r {parent=thigh_r, loc=<2.1908 -8.1142 1.7442>, rot=<0 0 0 -1>}
 deform_groin_r {parent=deform_groin_base_r, loc=<0 0 0>, rot=<0 0 0 1>}
 deform_glute_base_l {parent=pelvis, loc=<-8.8064 -8.8222 -8.4287>, rot=<0.0275 0.0501 -0.4472 -0.8926>}
 deform_glute_l {parent=deform_glute_base_l, loc=<0 0 0>, rot=<0 0 0 1>}
 deform_glute_base_r {parent=pelvis, loc=<-8.8062 -8.8222 8.4287>, rot=<0.0501 -0.0275 0.8926 -0.4472>}
 deform_glute_r {parent=deform_glute_base_r, loc=<0 0 0>, rot=<0 0 0 1>}
]
