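# Physics asset for the TikkaTaste medium male player cosmetic body, covering six dangling ring bones on the chest and back.
# Each ring has a capsule body and a constraint to its own small MSTR anchor body, with a larger spine_05 capsule for the upper torso.
# Constraints lock twist and one swing axis for hinge-like motion; the unlocked axis is swing1 on inner back rings and swing2 on the others.
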PhysicsAsset /BRCosmetics/Characters/Player/Male/Medium/Bodies/M_MED_TikkaTaste/Meshes/M_MED_TikkaTaste_Physics
bodies {
 dyn_chest_ring_l [{shapeType=Capsule, length=2.2221, radius=1.3839, shapeName=dyn_ring_back_out_l_capsule, center=<1.3271 0 0.6835>, rotation=<rot 0 0 90>}]
 dyn_chest_ring_l_MSTR [{shapeType=Capsule, length=0.505, radius=0.505, shapeName=dyn_chest_ring_l_MSTR_capsule, rotation=<rot 0 -90 0>}]
 dyn_chest_ring_r [{
   shapeType=Capsule
   length=2.2221
   radius=1.3839
   shapeName=dyn_ring_back_out_l_capsule
   center=<-1.2354 -0.0425 -0.8259>
   rotation=<rot 0 0 90>
  }]
 dyn_chest_ring_r_MSTR [{shapeType=Capsule, length=0.505, radius=0.505, shapeName=dyn_chest_ring_r_MSTR_capsule, rotation=<rot 0 -90 0>}]
 dyn_ring_back_in_l [{shapeType=Capsule, length=3, radius=2, shapeName=dyn_ring_back_in_l_capsule, center=<0.892 -1.522 0>}]
 dyn_ring_back_in_l_MSTR [{shapeType=Capsule, length=0.505, radius=0.505, shapeName=dyn_ring_back_in_l_MSTR_capsule, rotation=<rot 0 -90 0>}]
 dyn_ring_back_in_r [{shapeType=Capsule, length=3, radius=2, shapeName=dyn_ring_back_in_r_capsule, center=<-1.3166 1.4724 0>}]
 dyn_ring_back_in_r_MSTR [{shapeType=Capsule, length=0.505, radius=0.505, shapeName=dyn_ring_back_in_r_MSTR_capsule, rotation=<rot 0 -90 0>}]
 dyn_ring_back_out_l [{shapeType=Capsule, length=3, radius=2, shapeName=dyn_ring_back_out_l_capsule, center=<1.6138 0 -1.3163>, rotation=<rot 0 0 90>}]
 dyn_ring_back_out_l_MSTR [{shapeType=Capsule, length=0.505, radius=0.505, shapeName=dyn_ring_back_out_l_MSTR_capsule, rotation=<rot 0 -90 0>}]
 dyn_ring_back_out_r [{shapeType=Capsule, length=3, radius=2, shapeName=dyn_ring_back_out_r_capsule, center=<-1.801 0 1.2618>, rotation=<rot 0 0 90>}]
 dyn_ring_back_out_r_MSTR [{shapeType=Capsule, length=0.505, radius=0.505, shapeName=dyn_ring_back_out_r_MSTR_capsule, rotation=<rot 0 -90 0>}]
 spine_05 [{shapeType=Capsule, length=16.2424, radius=11.5066, shapeName=spine_05_capsule, rotation=<rot -0.108 -90 -0.0433>}]
}
constraints [
 {
  bone1Name=dyn_ring_back_in_l
  bone2Name=dyn_ring_back_in_l_MSTR
  swing1LimitDegrees=35
  swing2LimitDegrees=2
  swing2Motion=Locked
  twistLimitDegrees=45
  twistMotion=Locked
 }
 {
  bone1Name=dyn_ring_back_out_l
  bone2Name=dyn_ring_back_out_l_MSTR
  swing1LimitDegrees=2
  swing1Motion=Locked
  swing2LimitDegrees=35
  twistLimitDegrees=45
  twistMotion=Locked
 }
 {
  bone1Name=dyn_ring_back_in_r
  bone2Name=dyn_ring_back_in_r_MSTR
  swing1LimitDegrees=35
  swing2LimitDegrees=2
  swing2Motion=Locked
  twistLimitDegrees=45
  twistMotion=Locked
 }
 {
  bone1Name=dyn_ring_back_out_r
  bone2Name=dyn_ring_back_out_r_MSTR
  swing1LimitDegrees=2
  swing1Motion=Locked
  swing2LimitDegrees=35
  twistLimitDegrees=45
  twistMotion=Locked
 }
 {
  bone1Name=dyn_chest_ring_l
  bone2Name=dyn_chest_ring_l_MSTR
  swing1LimitDegrees=2
  swing1Motion=Locked
  swing2LimitDegrees=35
  twistLimitDegrees=45
  twistMotion=Locked
 }
 {
  bone1Name=dyn_chest_ring_r
  bone2Name=dyn_chest_ring_r_MSTR
  swing1LimitDegrees=2
  swing1Motion=Locked
  swing2LimitDegrees=35
  twistLimitDegrees=45
  twistMotion=Locked
 }
]
