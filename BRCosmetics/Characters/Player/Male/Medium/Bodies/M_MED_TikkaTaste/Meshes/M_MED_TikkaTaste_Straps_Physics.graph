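# Physics asset for the chest straps on the TikkaTaste medium male cosmetic body, defining the bodies and joints for their secondary motion.
# Six chains (top, middle and bottom, left and right) each link a small MSTR root capsule to two larger segment capsules.
# Constraints lock twist and swing1, leaving swing2 as the only unlocked angular axis, so segments bend like hinges.
# A large, unconstrained spine_02 capsule provides a torso volume.
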PhysicsAsset /BRCosmetics/Characters/Player/Male/Medium/Bodies/M_MED_TikkaTaste/Meshes/M_MED_TikkaTaste_Straps_Physics
bodies {
 dyn_chest_strap_btm_l_01 [{shapeType=Capsule, length=1.5, radius=1.2, shapeName=dyn_chest_strap_btm_l_01_capsule, center=<1.3531 0 0>, rotation=<rot 90 0 0>}]
 dyn_chest_strap_btm_l_02 [{shapeType=Capsule, length=1.5, radius=1.2, shapeName=dyn_chest_strap_btm_l_02_capsule, center=<1.5707 0 0>, rotation=<rot 90 0 0>}]
 dyn_chest_strap_btm_l_MSTR [{shapeType=Capsule, length=0.505, radius=0.505, shapeName=dyn_chest_strap_btm_l_MSTR_capsule, rotation=<rot 90 0 0>}]
 dyn_chest_strap_btm_r_01 [{shapeType=Capsule, length=1.5, radius=1.2, shapeName=dyn_chest_strap_btm_r_01_capsule, center=<-1.1892 0 0>, rotation=<rot 90 0 0>}]
 dyn_chest_strap_btm_r_02 [{shapeType=Capsule, length=1.5, radius=1.2, shapeName=dyn_chest_strap_btm_r_02_capsule, center=<-1.2541 0 0>, rotation=<rot 90 0 0>}]
 dyn_chest_strap_btm_r_MSTR [{shapeType=Capsule, length=0.505, radius=0.505, shapeName=dyn_chest_strap_btm_r_MSTR_capsule, rotation=<rot 90 0 0>}]
 dyn_chest_strap_mid_l_01 [{shapeType=Capsule, length=1.5, radius=1.2, shapeName=dyn_chest_strap_mid_l_01_capsule, center=<1.3531 0 0>, rotation=<rot 90 0 0>}]
 dyn_chest_strap_mid_l_02 [{shapeType=Capsule, length=1.5, radius=1.2, shapeName=dyn_chest_strap_mid_l_02_capsule, center=<1.5707 0 0>, rotation=<rot 90 0 0>}]
 dyn_chest_strap_mid_l_MSTR [{shapeType=Capsule, length=0.505, radius=0.505, shapeName=dyn_chest_strap_mid_l_MSTR_capsule, rotation=<rot 90 0 0>}]
 dyn_chest_strap_mid_r_01 [{shapeType=Capsule, length=1.5, radius=1.2, shapeName=dyn_chest_strap_mid_r_01_capsule, center=<-1.1892 0 0>, rotation=<rot 90 0 0>}]
 dyn_chest_strap_mid_r_02 [{shapeType=Capsule, length=1.5, radius=1.2, shapeName=dyn_chest_strap_mid_r_02_capsule, center=<-1.2541 0 0>, rotation=<rot 90 0 0>}]
 dyn_chest_strap_mid_r_MSTR [{shapeType=Capsule, length=0.505, radius=0.505, shapeName=dyn_chest_strap_mid_r_MSTR_capsule, rotation=<rot 90 0 0>}]
 dyn_chest_strap_top_l_01 [{shapeType=Capsule, length=1.5, radius=1.2, shapeName=dyn_chest_strap_top_l_01_capsule, center=<1.3531 0 0>, rotation=<rot 90 0 0>}]
 dyn_chest_strap_top_l_02 [{shapeType=Capsule, length=1.5, radius=1.2, shapeName=dyn_chest_strap_top_l_02_capsule, center=<1.5707 0 0>, rotation=<rot 90 0 0>}]
 dyn_chest_strap_top_l_MSTR [{shapeType=Capsule, length=0.505, radius=0.505, shapeName=dyn_chest_strap_top_l_MSTR_capsule, rotation=<rot 90 0 0>}]
 dyn_chest_strap_top_r_01 [{shapeType=Capsule, length=1.5, radius=1.2, shapeName=dyn_chest_strap_top_r_01_capsule, center=<-1.1892 0 0>, rotation=<rot 90 0 0>}]
 dyn_chest_strap_top_r_02 [{shapeType=Capsule, length=1.5, radius=1.2, shapeName=dyn_chest_strap_top_r_02_capsule, center=<-1.2541 0 0>, rotation=<rot 90 0 0>}]
 dyn_chest_strap_top_r_MSTR [{shapeType=Capsule, length=0.505, radius=0.505, shapeName=dyn_chest_strap_top_r_MSTR_capsule, rotation=<rot 90 0 0>}]
 spine_02 [{shapeType=Capsule, length=0.441, radius=14.5297, shapeName=spine_02_capsule, center=<4.8211 0 0.0061>, rotation=<rot -0.1277 -90 -0.073>}]
}
constraints [
 {
  bone1Name=dyn_chest_strap_btm_l_01
  bone2Name=dyn_chest_strap_btm_l_MSTR
  swing1LimitDegrees=45
  swing1Motion=Locked
  swing2LimitDegrees=10
  twistLimitDegrees=45
  twistMotion=Locked
 }
 {
  bone1Name=dyn_chest_strap_btm_l_02
  bone2Name=dyn_chest_strap_btm_l_01
  swing1LimitDegrees=45
  swing1Motion=Locked
  swing2LimitDegrees=10
  twistLimitDegrees=45
  twistMotion=Locked
 }
 {
  bone1Name=dyn_chest_strap_btm_r_01
  bone2Name=dyn_chest_strap_btm_r_MSTR
  swing1LimitDegrees=45
  swing1Motion=Locked
  swing2LimitDegrees=10
  twistLimitDegrees=45
  twistMotion=Locked
 }
 {
  bone1Name=dyn_chest_strap_btm_r_02
  bone2Name=dyn_chest_strap_btm_r_01
  swing1LimitDegrees=45
  swing1Motion=Locked
  swing2LimitDegrees=10
  twistLimitDegrees=45
  twistMotion=Locked
 }
 {
  bone1Name=dyn_chest_strap_mid_l_01
  bone2Name=dyn_chest_strap_mid_l_MSTR
  swing1LimitDegrees=45
  swing1Motion=Locked
  swing2LimitDegrees=10
  twistLimitDegrees=45
  twistMotion=Locked
 }
 {
  bone1Name=dyn_chest_strap_mid_l_02
  bone2Name=dyn_chest_strap_mid_l_01
  swing1LimitDegrees=45
  swing1Motion=Locked
  swing2LimitDegrees=10
  twistLimitDegrees=45
  twistMotion=Locked
 }
 {
  bone1Name=dyn_chest_strap_mid_r_01
  bone2Name=dyn_chest_strap_mid_r_MSTR
  swing1LimitDegrees=45
  swing1Motion=Locked
  swing2LimitDegrees=10
  twistLimitDegrees=45
  twistMotion=Locked
 }
 {
  bone1Name=dyn_chest_strap_mid_r_02
  bone2Name=dyn_chest_strap_mid_r_01
  swing1LimitDegrees=45
  swing1Motion=Locked
  swing2LimitDegrees=10
  twistLimitDegrees=45
  twistMotion=Locked
 }
 {
  bone1Name=dyn_chest_strap_top_l_01
  bone2Name=dyn_chest_strap_top_l_MSTR
  swing1LimitDegrees=45
  swing1Motion=Locked
  swing2LimitDegrees=10
  twistLimitDegrees=45
  twistMotion=Locked
 }
 {
  bone1Name=dyn_chest_strap_top_l_02
  bone2Name=dyn_chest_strap_top_l_01
  swing1LimitDegrees=45
  swing1Motion=Locked
  swing2LimitDegrees=10
  twistLimitDegrees=45
  twistMotion=Locked
 }
 {
  bone1Name=dyn_chest_strap_top_r_01
  bone2Name=dyn_chest_strap_top_r_MSTR
  swing1LimitDegrees=45
  swing1Motion=Locked
  swing2LimitDegrees=10
  twistLimitDegrees=45
  twistMotion=Locked
 }
 {
  bone1Name=dyn_chest_strap_top_r_02
  bone2Name=dyn_chest_strap_top_r_01
  swing1LimitDegrees=45
  swing1Motion=Locked
  swing2LimitDegrees=10
  twistLimitDegrees=45
  twistMotion=Locked
 }
]
