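# Animation blueprint for the TikkaTaste medium male player body cosmetic, adding secondary physics motion to a pose copied from the attached parent mesh.
# The copied pose runs through a Control Rig, then two RigidBody nodes with separate body and strap physics assets, simulated relative to spine bones.
# Spring-driven AnimDynamics nodes then swing both zippers and both three-bone boot strap chains with angular limits.
# The event graph has no logic.
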AnimBlueprint /BRCosmetics/Characters/Player/Male/Medium/Bodies/M_MED_TikkaTaste/Meshes/M_MED_TikkaTaste_AnimBP
target_skeleton=&/M_MED_TikkaTaste_Skeleton
graphs [
 AnimGraph {nodes [
   n1 = Control Rig(Source=Copy Pose From Mesh() {bUseAttachedParent=true}) {controlRigAssetReference {blueprintRigClass=&/M_MED_TikkaTaste_CtrlRig.M_MED_TikkaTaste_CtrlRig_C}}
   n2 = RigidBody(ComponentPose=Local To Component(LocalPose=@n1)) {
    bClampLinearTranslationLimitToRefPose=true
    bOverrideWorldGravity=true
    overridePhysicsAsset=&/M_MED_TikkaTaste_Physics
    simulationSpace=BaseBoneSpace
    baseBoneRef {boneName=spine_04}
    overrideWorldGravity {z=-980}
    simSpaceSettings {
     dampingAlpha=0
     maxAngularAcceleration=30
     maxAngularVelocity=6
     maxLinearAcceleration=3000
     maxLinearVelocity=600
     velocityScaleZ=0.75
     worldAlpha=1
     externalLinearDragV=<0.15 0.15 0.15>
    }
   }
   n3 = RigidBody(ComponentPose=@n2) {
    bClampLinearTranslationLimitToRefPose=true
    bOverrideWorldGravity=true
    overridePhysicsAsset=&/M_MED_TikkaTaste_Straps_Physics
    simulationSpace=BaseBoneSpace
    baseBoneRef {boneName=spine_02}
    overrideWorldGravity {z=-980}
    simSpaceSettings {
     dampingAlpha=0
     maxAngularAcceleration=30
     maxAngularVelocity=6
     maxLinearAcceleration=2000
     maxLinearVelocity=400
     worldAlpha=1
     externalLinearDragV=<0.2 0.2 0.2>
    }
   }
   n4 = Anim Dynamics - Bone: dyn_zipper_l(ComponentPose=@n3, Alpha=1, ExternalForce="0,0,0") {
    angularSpringConstant=200
    bAngularSpring=true
    bUseGravityOverride=true
    simulationSpace=RootRelative
    boundBone {boneName=dyn_zipper_l}
    componentAppliedLinearAccClamp=<3000 3000 3000>
    componentLinearAccScale=<0.3 0.3 0.3>
    componentLinearVelScale=<0.3 0.3 0.3>
    gravityOverride {z=-980}
    physicsBodyDefinitions [{
      boundBone {boneName=dyn_zipper_l}
      boxExtents=<10 15 2>
      constraintSetup {angularLimitsMax=<0 50 5>, angularLimitsMin=<0 0 -5>, angularTarget=<1 0 0>}
      localJointOffset=<3 0 0>
     }]
    relativeSpaceBone {boneName=thigh_twist_01_l}
    simSpaceSettings {maxAngularAcceleration=60, maxAngularVelocity=12, simSpaceAngularAlpha=0.4}
   }
   n5 = Anim Dynamics - Bone: dyn_zipper_r(ComponentPose=@n4, Alpha=1, ExternalForce="0,0,0") {
    angularSpringConstant=200
    bAngularSpring=true
    bUseGravityOverride=true
    simulationSpace=RootRelative
    boundBone {boneName=dyn_zipper_r}
    componentAppliedLinearAccClamp=<3000 3000 3000>
    componentLinearAccScale=<0.3 0.3 0.3>
    componentLinearVelScale=<0.3 0.3 0.3>
    gravityOverride {z=-980}
    physicsBodyDefinitions [{
      boundBone {boneName=dyn_zipper_r}
      boxExtents=<10 15 2>
      constraintSetup {angularLimitsMax=<0 50 5>, angularLimitsMin=<0 0 -5>, angularTarget=<1 0 0>}
      localJointOffset=<-3 0 0>
     }]
    relativeSpaceBone {boneName=thigh_twist_01_l}
    simSpaceSettings {maxAngularAcceleration=60, maxAngularVelocity=12, simSpaceAngularAlpha=0.4}
   }
   n6 = "Anim Dynamics - Chain: dyn_boot_strap_l_01 -> dyn_boot_strap_l_03"(ComponentPose=@n5, Alpha=1, ExternalForce="0,0,0") {
    angularDampingOverride=0.999
    angularSpringConstant=2000
    bAngularSpring=true
    bChain=true
    bOverrideAngularDamping=true
    bUseGravityOverride=true
    simulationSpace=RootRelative
    boundBone {boneName=dyn_boot_strap_l_01}
    chainEnd {boneName=dyn_boot_strap_l_03}
    componentAppliedLinearAccClamp=<3000 3000 3000>
    componentLinearAccScale=<0.3 0.3 0.3>
    componentLinearVelScale=<0.3 0.3 0.3>
    gravityOverride {z=-980}
    physicsBodyDefinitions [
     {
      boundBone {boneName=dyn_boot_strap_l_01}
      boxExtents=<4 15 15>
      constraintSetup {angularLimitsMax=<0 25 0>, angularLimitsMin=<0 -5 0>, angularTarget=<1 0 0>}
      localJointOffset=<0.1 -1 0>
     }
     {
      boundBone {boneName=dyn_boot_strap_l_02}
      boxExtents=<4 15 15>
      constraintSetup {angularLimitsMax=<0 25 0>, angularLimitsMin=<0 -5 0>, angularTarget=<1 0 0>}
      localJointOffset=<0.1 -1 0>
     }
     {
      boundBone {boneName=dyn_boot_strap_l_03}
      boxExtents=<4 15 15>
      constraintSetup {angularLimitsMax=<0 25 0>, angularLimitsMin=<0 -5 0>, angularTarget=<1 0 0>}
      localJointOffset=<0.1 -1 0>
     }
    ]
    relativeSpaceBone {boneName=thigh_twist_01_l}
    simSpaceSettings {maxAngularAcceleration=60, maxAngularVelocity=12, simSpaceAngularAlpha=0.4}
   }
   n7 = "Anim Dynamics - Chain: dyn_boot_strap_r_01 -> dyn_boot_strap_r__03"(ComponentPose=@n6, Alpha=1, ExternalForce="0,0,0") {
    angularDampingOverride=0.999
    angularSpringConstant=2000
    bAngularSpring=true
    bChain=true
    bOverrideAngularDamping=true
    bOverrideLinearDamping=true
    bUseGravityOverride=true
    linearDampingOverride=0.9
    simulationSpace=RootRelative
    boundBone {boneName=dyn_boot_strap_r_01}
    chainEnd {boneName=dyn_boot_strap_r__03}
    componentAppliedLinearAccClamp=<3000 3000 3000>
    componentLinearAccScale=<0.3 0.3 0.3>
    componentLinearVelScale=<0.3 0.3 0.3>
    gravityOverride {z=-980}
    physicsBodyDefinitions [
     {
      boundBone {boneName=dyn_boot_strap_r_01}
      boxExtents=<4 15 15>
      constraintSetup {angularLimitsMax=<0 25 0>, angularLimitsMin=<0 -5 0>, angularTarget=<1 0 0>}
      localJointOffset=<-0.1 -1 0>
     }
     {
      boundBone {boneName=dyn_boot_strap_r__02}
      boxExtents=<4 15 15>
      constraintSetup {angularLimitsMax=<0 25 0>, angularLimitsMin=<0 -5 0>, angularTarget=<1 0 0>}
      localJointOffset=<-0.1 -1 0>
     }
     {
      boundBone {boneName=dyn_boot_strap_r__03}
      boxExtents=<4 15 15>
      constraintSetup {angularLimitsMax=<0 25 0>, angularLimitsMin=<0 -5 0>, angularTarget=<1 0 0>}
      localJointOffset=<-0.1 -1 0>
     }
    ]
    relativeSpaceBone {boneName=thigh_twist_01_l}
    simSpaceSettings {maxAngularAcceleration=60, maxAngularVelocity=12, simSpaceAngularAlpha=0.4}
   }
   Output Pose(Result=Component To Local(ComponentPose=@n7)) {layerGroup=None}
  ]}
 EventGraph {nodes [Event BlueprintUpdateAnimation()]}
]
variables []
