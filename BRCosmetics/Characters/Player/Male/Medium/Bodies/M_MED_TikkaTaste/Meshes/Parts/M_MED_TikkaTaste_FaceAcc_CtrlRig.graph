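# Control Rig for the TikkaTaste male medium body's accessory part, positioning mag, chest accessory and sheath bones as the body moves.
# Weighted multi-parent constraints blend those bones from spine, pelvis, clavicle and body deform bones.
# BaseballSliding and KneeSliding bools then set 0/1 alphas, which are interpolated to blend in a separate rotation offset on dyn_sheath_MSTR for each slide.
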ControlRigBlueprint /BRCosmetics/Characters/Player/Male/Medium/Bodies/M_MED_TikkaTaste/Meshes/Parts/M_MED_TikkaTaste_FaceAcc_CtrlRig
graphs [
 RigVMFunctionLibrary {links [], nodes []}
 RigVMModel {nodes [
   RigUnit_BeginExecution = &RigUnit_BeginExecution()
   RigVMFunction_Sequence = &RigVMFunction_Sequence(ExecuteContext=@RigUnit_BeginExecution.ExecutePin)
   &RigUnit_ParentConstraint(
    AdvancedSettings=(RotationOrderForFilter=XZY)
    Child=(Type=Bone,Name="deform_mag_btm_l")
    Parents=((Item=(Type=Bone,Name="spine_04"),Weight=0.0326),(Item=(Type=Bone,Name="pelvis"),Weight=0.1618),(Item=(Type=Bone,Name="spine_01"),Weight=0.2083),(Item=(Type=Bone,Name="spine_03"),Weight=0.2564),(Item=(Type=Bone,Name="spine_02"),Weight=0.3409))
    Weight=1
    ExecutePin=@RigVMFunction_Sequence.A
   )
   > &RigUnit_ParentConstraint(
    AdvancedSettings=(RotationOrderForFilter=XZY)
    Child=(Type=Bone,Name="deform_mag_btm_r")
    Parents=((Item=(Type=Bone,Name="spine_04"),Weight=0.0326),(Item=(Type=Bone,Name="pelvis"),Weight=0.1618),(Item=(Type=Bone,Name="spine_01"),Weight=0.2083),(Item=(Type=Bone,Name="spine_03"),Weight=0.2564),(Item=(Type=Bone,Name="spine_02"),Weight=0.3409))
    Weight=1
   )
   > &RigUnit_ParentConstraint(
    AdvancedSettings=(RotationOrderForFilter=XZY)
    Child=(Type=Bone,Name="deform_mag_top_r")
    Parents="((Item=(Type=Bone,Name=\"pelvis\"),Weight=0.0072),(Item=(Type=Bone,Name=\"upperarm_twist_01_r\"),Weight=0.0129),(Item=(Type=Bone,Name=\"spine_01\"),Weight=0.0592),(Item=(Type=Bone,Name=\"clavicle_r\"),Weight=0.1174),(Item=(Type=Bone,Name=\"deform_pec_l\"),Weight=0.1 ...(truncated)"
    Weight=1
   )
   > &RigUnit_ParentConstraint(
    AdvancedSettings=(RotationOrderForFilter=XZY)
    Child=(Type=Bone,Name="deform_mag_top_l")
    Parents="((Item=(Type=Bone,Name=\"pelvis\"),Weight=0.0072),(Item=(Type=Bone,Name=\"upperarm_twist_01_l\"),Weight=0.0129),(Item=(Type=Bone,Name=\"spine_01\"),Weight=0.0592),(Item=(Type=Bone,Name=\"clavicle_l\"),Weight=0.1174),(Item=(Type=Bone,Name=\"deform_pec_l\"),Weight=0.1 ...(truncated)"
    Weight=1
   )
   > &RigUnit_ParentConstraint(
    AdvancedSettings=(RotationOrderForFilter=XZY)
    Child=(Type=Bone,Name="dyn_sheath_MSTR")
    Parents="((Item=(Type=Bone,Name=\"deform_groin_l\"),Weight=0.0072),(Item=(Type=Bone,Name=\"spine_03\"),Weight=0.0133),(Item=(Type=Bone,Name=\"deform_glute_l\"),Weight=0.0193),(Item=(Type=Bone,Name=\"spine_01\"),Weight=0.0209),(Item=(Type=Bone,Name=\"spine_02\"),Weight=0.0242 ...(truncated)"
    Weight=1
   )
   > &RigUnit_ParentConstraint(
    AdvancedSettings=(RotationOrderForFilter=XZY)
    Child=(Type=Bone,Name="deform_chest_acc_MSTR")
    Parents="((Item=(Type=Bone,Name=\"clavicle_r\"),Weight=0.0016),(Item=(Type=Bone,Name=\"clavicle_l\"),Weight=0.0016),(Item=(Type=Bone,Name=\"deform_pec_l\"),Weight=0.0041),(Item=(Type=Bone,Name=\"spine_02\"),Weight=0.0747),(Item=(Type=Bone,Name=\"spine_03\"),Weight=0.1876),(I ...(truncated)"
    Weight=1
   )
   n1 = &RigVMFunction_Sequence(ExecuteContext=@RigVMFunction_Sequence.B)
   RigVMFunction_ControlFlowBranch = &RigVMFunction_ControlFlowBranch(ExecuteContext=@n1.A, Condition=RigVMVariableNode() {variable=BaseballSliding}.Value)
   RigVMVariableNode(Value=0, ExecuteContext=@RigVMFunction_ControlFlowBranch.False) {variable=BaseballSlideAlpha}
   RigVMVariableNode(Value=1, ExecuteContext=@RigVMFunction_ControlFlowBranch.True) {variable=BaseballSlideAlpha}
   AlphaInterp = &RigVMFunction_AlphaInterp(
    InterpSpeedDecreasing=10
    InterpSpeedIncreasing=10
    Scale=1
    bClampResult=true
    Value=RigVMVariableNode() {variable=BaseballSlideAlpha}.Value
   )
   &RigUnit_ModifyTransforms(
    ItemToModify=((Item=(Type=Bone,Name="dyn_sheath_MSTR"),Transform=(Rotation=(X=0,Y=0.4491,Z=0,W=-0.8935),Translation=(X=0,Y=0,Z=0),Scale3D=(X=1,Y=1,Z=1))))
    WeightMinimum=0
    Weight=@AlphaInterp.Result
    ExecutePin=@RigVMFunction_ControlFlowBranch.Completed
   )
   n2 = &RigVMFunction_ControlFlowBranch(ExecuteContext=@n1.B, Condition=RigVMVariableNode() {variable=KneeSliding}.Value)
   n3 = &RigVMFunction_AlphaInterp(InterpSpeedDecreasing=10, InterpSpeedIncreasing=10, Scale=1, bClampResult=true, Value=RigVMVariableNode() {variable=KneeSlideAlpha}.Value)
   &RigUnit_ModifyTransforms(
    ItemToModify=((Item=(Type=Bone,Name="dyn_sheath_MSTR"),Transform=(Rotation=(X=0.2999,Y=-0.6693,Z=0.4322,W=-0.5247),Translation=(X=0,Y=0,Z=0),Scale3D=(X=1,Y=1,Z=1))))
    WeightMinimum=0
    Weight=@n3.Result
    ExecutePin=@n2.Completed
   )
   RigVMVariableNode(ExecuteContext=@n2.False) {variable=KneeSlideAlpha}
   RigVMVariableNode(Value=1, ExecuteContext=@n2.True) {variable=KneeSlideAlpha}
  ]}
]
hierarchy [
 root {loc=<0 0 0>, rot=<0 0 0 1>}
 pelvis {parent=root, loc=<0 -0.7071 93.1191>, rot=<0.0127 0.707 -0.0127 -0.707>}
 spine_01 {parent=pelvis, loc=<6.7494 -0.6388 0>, rot=<0 0 0.034 -0.9994>}
 spine_02 {parent=spine_01, loc=<7.2518 -0.035 0>, rot=<0 0 0.0238 -0.9997>}
 spine_03 {parent=spine_02, loc=<6.0104 0 0>, rot=<0 0 0.0149 -0.9999>}
 spine_04 {parent=spine_03, loc=<8.1024 0 0>, rot=<0 0 -0.0132 -0.9999>}
 spine_05 {parent=spine_04, loc=<10.0146 0 0>, rot=<0 0 -0.0015 -1>}
 neck_01 {parent=spine_05, loc=<11.3138 0.1529 0>, rot=<0 0 -0.2003 -0.9797>}
 neck_02 {parent=neck_01, loc=<4.5627 0 0>, rot=<0 0 0.0504 -0.9987>}
 head {parent=neck_02, loc=<5.3914 0 0>, rot=<0 0 0.115 -0.9934>}
 clavicle_l {parent=spine_05, loc=<4.3575 -0.1466 -2.3223>, rot=<0.0465 -0.7289 -0.009 -0.6829>}
 upperarm_l {parent=clavicle_l, loc=<15.5111 0 0>, rot=<0.0226 -0.3234 -0.0088 -0.9459>}
 upperarm_twist_01_l {parent=upperarm_l, loc=<0.0041 0 0>, rot=<0 0 0 -1>}
 clavicle_r {parent=spine_05, loc=<4.3575 -0.1466 2.3223>, rot=<0.7289 0.0465 -0.6829 0.009>}
 upperarm_r {parent=clavicle_r, loc=<-15.5111 0 0>, rot=<0.0226 -0.3234 -0.0088 -0.9459>}
 upperarm_twist_01_r {parent=upperarm_r, loc=<-0.0041 0 0>, rot=<0 0 0 -1>}
 deform_chest_acc_MSTR {parent=spine_05, loc=<-0.5789 14.5888 -0.0109>, rot=<0.0284 -0.7065 -0.0284 -0.7065>}
 deform_chest_acc {parent=deform_chest_acc_MSTR, loc=<0 0 0>, rot=<0 0 0 1>}
 deform_pec_aim_base_l {parent=spine_04, loc=<7.8732 12.6388 -2.4123>, rot=<0.0102 -0.4602 0.3979 -0.7936>}
 deform_pec_aim_l {parent=deform_pec_aim_base_l, loc=<0 0 0>, rot=<0 0 0 1>}
 deform_pec_base_l {parent=deform_pec_aim_l, loc=<10.766 0 0>, rot=<0 0 0 1>}
 deform_pec_l {parent=deform_pec_base_l, loc=<0.055 0 0>, rot=<0 0 0 -1>}
 deform_pec_aim_base_r {parent=spine_04, loc=<7.8727 12.6388 2.4123>, rot=<-0.4602 -0.0102 0.7936 0.3979>}
 deform_pec_aim_r {parent=deform_pec_aim_base_r, loc=<0 0 0>, rot=<0 0 0 1>}
 deform_pec_base_r {parent=deform_pec_aim_r, loc=<-10.7662 -0.0006 -0.0007>, rot=<0 0 0 1>}
 deform_pec_r {parent=deform_pec_base_r, loc=<0 0 0>, rot=<0 0 0 -1>}
 deform_mag_top_l {parent=spine_03, loc=<6.9971 10.187 -13.898>, rot=<0.0387 -0.706 -0.0387 -0.706>}
 deform_mag_top_r {parent=spine_03, loc=<6.9971 10.187 13.898>, rot=<0.0387 -0.706 -0.0387 -0.706>}
 deform_mag_btm_l {parent=spine_02, loc=<2.4871 12.3448 -9.5111>, rot=<0.0282 -0.7065 -0.0282 -0.7065>}
 deform_mag_btm_r {parent=spine_02, loc=<2.4871 12.3448 9.5111>, rot=<0.0282 -0.7065 -0.0282 -0.7065>}
 dyn_sheath_MSTR {parent=pelvis, loc=<5.233 -3.998 -13.9921>, rot=<0.1589 -0.8702 -0.4555 -0.1002>}
 dyn_sheath_01 {parent=dyn_sheath_MSTR, loc=<0 0 0>, rot=<0 0 0 1>}
 dyn_sheath_02 {parent=dyn_sheath_01, loc=<2.4523 0 0>, rot=<-0.042 -0.0188 0.0237 -0.9987>}
 dyn_sheath_03 {parent=dyn_sheath_02, loc=<1.9696 0 0>, rot=<0 0 -0.1478 -0.989>}
 dyn_sheath_ring_01 {parent=dyn_sheath_03, loc=<4.6229 -0.3747 3.6928>, rot=<0.6452 0.6434 0.302 -0.2803>}
 dyn_sheath_ring_02 {parent=dyn_sheath_03, loc=<27.7543 -0.0264 3.8609>, rot=<-0.66 -0.6781 -0.237 0.2199>}
 thigh_l {parent=pelvis, loc=<-2.5693 -0.0109 -7.7501>, rot=<0.0012 0.0572 0.0155 -0.9982>}
 deform_groin_base_l {parent=thigh_l, loc=<-2.1908 8.1142 -1.7442>, rot=<0 0 0 -1>}
 deform_groin_l {parent=deform_groin_base_l, loc=<0 0 0>, rot=<0 0 0 -1>}
 calf_l {parent=thigh_l, loc=<-39.0484 0 0>, rot=<0.0006 -0.0161 -0.0391 -0.9991>}
 thigh_twist_01_l {parent=thigh_l, loc=<-21.6303 0 0>, rot=<0.0013 0 0 -1>}
 deform_glute_base_l {parent=pelvis, loc=<-8.8064 -8.8222 -8.4287>, rot=<0.0275 0.0501 -0.4472 -0.8926>}
 deform_glute_l {parent=deform_glute_base_l, loc=<0 0 0>, rot=<0 0 0 1>}
]
variables [
 BaseballSlideAlpha {type=double, default=0}
 BaseballSliding {type=bool, default=False}
 KneeSlideAlpha {type=double, default=0}
 KneeSliding {type=bool, default=False}
]
