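# Physics asset for the TikkaTaste medium male FaceAcc part, rigging a dynamic sheath beside the pelvis and left leg.
# All bodies are single capsules, with a large one on the last sheath segment and unconstrained ones on pelvis, left thigh, left calf and the sheath CONST bone.
# Constraints chain three sheath segments from the MSTR bone with small angle limits, then branch to two twist-locked rings on the last segment.
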PhysicsAsset /BRCosmetics/Characters/Player/Male/Medium/Bodies/M_MED_TikkaTaste/Meshes/Parts/M_MED_TikkaTaste_FaceAcc_Physics
bodies {
 calf_l [{shapeType=Capsule, length=40.2998, radius=9.1329, shapeName=calf_l_capsule, center=<-21.9941 -1.5796 0.7053>, rotation=<rot 90 0 0>}]
 dyn_sheath_01 [{shapeType=Capsule, length=1.7114, radius=0.9848, shapeName=dyn_sheath_01_capsule, center=<1.5 -0.6132 0>, rotation=<rot 90 0 0>}]
 dyn_sheath_02 [{shapeType=Capsule, length=1.7114, radius=0.9848, shapeName=dyn_sheath_02_capsule, center=<1.5 -0.6132 0>, rotation=<rot 90 0 0>}]
 dyn_sheath_03 [{
   shapeType=Capsule
   length=42.8628
   radius=7.7882
   shapeName=dyn_sheath_03_capsule
   center=<24.3041 -7.1472 0.7768>
   rotation=<rot -0.0396 91.6711 -88.8306>
  }]
 dyn_sheath_CONST [{
   shapeType=Capsule
   length=1.7175
   radius=1.5649
   shapeName=dyn_sheath_CONST_capsule
   center=<-0.1561 -0.1329 0.1334>
   rotation=<rot 0 -85.1583 -72.7957>
  }]
 dyn_sheath_MSTR [{shapeType=Capsule, length=0.505, radius=0.505, shapeName=dyn_sheath_MSTR_capsule, rotation=<rot 0 -90 0>}]
 dyn_sheath_ring_01 [{shapeType=Capsule, length=4.3158, radius=2.5236, shapeName=dyn_sheath_ring_01_capsule, center=<4.2318 0 -1.63>, rotation=<rot 90 0 0>}]
 dyn_sheath_ring_02 [{shapeType=Capsule, length=4.3642, radius=2.5498, shapeName=dyn_sheath_ring_02_capsule, center=<4.2318 0 -2.2554>, rotation=<rot 90 0 0>}]
 pelvis [{shapeType=Capsule, length=2.3662, radius=14.3249, shapeName=pelvis_capsule}]
 thigh_l [{
   shapeType=Capsule
   length=45.0306
   radius=11.4478
   shapeName=thigh_l_capsule
   center=<-21.4375 -1.4041 3.1218>
   rotation=<rot 72.8457 88.584 85.4457>
  }]
}
constraints [
 {bone1Name=dyn_sheath_01, bone2Name=dyn_sheath_MSTR, swing1LimitDegrees=8, swing2LimitDegrees=4, twistLimitDegrees=4}
 {bone1Name=dyn_sheath_02, bone2Name=dyn_sheath_01, swing1LimitDegrees=8, swing2LimitDegrees=4, twistLimitDegrees=4}
 {bone1Name=dyn_sheath_03, bone2Name=dyn_sheath_02, swing1LimitDegrees=8, swing2LimitDegrees=4, twistLimitDegrees=4}
 {bone1Name=dyn_sheath_ring_01, bone2Name=dyn_sheath_03, swing1LimitDegrees=2, swing2LimitDegrees=24, twistLimitDegrees=45, twistMotion=Locked}
 {bone1Name=dyn_sheath_ring_02, bone2Name=dyn_sheath_03, swing1LimitDegrees=2, swing2LimitDegrees=24, twistLimitDegrees=45, twistMotion=Locked}
]
