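# Animation Blueprint for the TikkaTaste male medium player body's face accessory part, following the character's pose with secondary motion.
# Layers Copy Pose From Mesh from the customization-loaded base mesh and the attached parent via blend per bone.
# A Control Rig receives baseball and knee slide flags set by events.
# RigidBody then simulates physics in pelvis bone space with overridden gravity and capped simulation-space motion.
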AnimBlueprint /BRCosmetics/Characters/Player/Male/Medium/Bodies/M_MED_TikkaTaste/Meshes/Parts/M_MED_TikkaTaste_FaceAcc_AnimBP
target_skeleton=&/M_MED_TikkaTaste_FaceAcc_Skeleton
graphs [
 AnimGraph {nodes [
   n1 = Get CharacterMesh() -> (CharacterMesh:Skeletal Mesh Component Object Reference)
   n2 = Control Rig!(
    Source=Layered blend per bone(
     BasePose=Copy Pose From Mesh(SourceMeshComponent=@n1) {bUseAttachedParent=true}
     BlendPoses_0=Copy Pose From Mesh() {bUseAttachedParent=true}
    )
    Alpha
    BaseballSliding:bool=Get BaseballSliding() -> (BaseballSliding:bool)
    KneeSliding:bool=Get KneeSliding() -> (KneeSliding:bool)
   ) -> (Pose) {controlRigAssetReference {blueprintRigClass=&/M_MED_TikkaTaste_FaceAcc_CtrlRig.M_MED_TikkaTaste_FaceAcc_CtrlRig_C}}
   n3 = RigidBody(ComponentPose=Local To Component(LocalPose=@n2)) {
    bClampLinearTranslationLimitToRefPose=true
    bOverrideWorldGravity=true
    overridePhysicsAsset=&/M_MED_TikkaTaste_FaceAcc_Physics
    simulationSpace=BaseBoneSpace
    baseBoneRef {boneName=pelvis}
    overrideWorldGravity {z=-980}
    simSpaceSettings {
     dampingAlpha=0
     maxAngularAcceleration=30
     maxAngularVelocity=6
     maxLinearAcceleration=4000
     maxLinearVelocity=800
     worldAlpha=1
     externalLinearDragV=<0.3 0.3 0.3>
    }
   }
   Output Pose(Result=Component To Local(ComponentPose=@n3)) {layerGroup=None}
  ]}
 EventGraph {nodes [
   Event OnBeginBaseballSliding() {is=title K2Node_Event}
   > Set BaseballSliding(BaseballSliding:bool=true) -> (Output_Get:bool)
   n1 = Event OnCharacterCustomizationLoaded()
   > Set CharacterMesh(CharacterMesh:Skeletal Mesh Component Object Reference=@n1.CustomBaseMesh) -> (Output_Get:Skeletal Mesh Component Object Reference)
   Event OnEndBaseballSliding() {is=title K2Node_Event}
   > Set BaseballSliding(BaseballSliding=false)
   Event OnBeginKneeSliding() {is=title K2Node_Event}
   > Set KneeSliding(KneeSliding:bool=true) -> (Output_Get:bool)
   Event OnEndKneeSliding() {is=title K2Node_Event}
   > Set KneeSliding(KneeSliding=false)
   Event BlueprintUpdateAnimation()
  ]}
]
variables [
 BaseballSliding {type=boolean}
 CharacterMesh {type=SkeletalMeshComponent}
 KneeSliding {type=boolean}
]
